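# Sound cue for a looping quad ambient bed heard inside a group of Curie fires, routed to the Athena ambient sound class.
# A Random node picks one of four equally weighted looping waves, then a Modulator slightly varies its pitch.
# It uses large-group fire attenuation and a concurrency override that allows a single instance and blocks new ones while it plays.
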SoundCue /Game/Sounds/Fort_GamePlay_Sounds/Curie_Fires/Curie_Fire_GroupInsideLoop_Quad_Cue
root=Modulator_0
nodes [
 {id=Modulator_0, children [Random_1], properties {pitchMax=1.1, pitchMin=0.9}}
 {id=Random_1, children [WavePlayer_6,WavePlayer_7,WavePlayer_8,WavePlayer_9], properties {weights [1,1,1,1]}}
 {
  id=WavePlayer_6
  properties {bLooping=true, soundWaveAssetPtr=&/Curie_Fire_GroupInsideLoop_Quad_01}
 }
 {
  id=WavePlayer_7
  properties {bLooping=true, soundWaveAssetPtr=&/Curie_Fire_GroupInsideLoop_Quad_02}
 }
 {
  id=WavePlayer_8
  properties {bLooping=true, soundWaveAssetPtr=&/Curie_Fire_GroupInsideLoop_Quad_03}
 }
 {
  id=WavePlayer_9
  properties {bLooping=true, soundWaveAssetPtr=&/Curie_Fire_GroupInsideLoop_Quad_04}
 }
]
properties {
 attenuationSettings=&/Curie_Fire_GroupLarge_01_Att
 bOverrideConcurrency=true
 maxDistance=5200
 soundClassObject=&Fort_Ambient_Athena
 volumeMultiplier=0.752
 concurrencyOverrides {maxCount=1, resolutionRule=PreventNew}
}
