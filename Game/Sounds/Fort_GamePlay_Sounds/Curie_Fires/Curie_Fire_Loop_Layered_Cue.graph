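# Curie fire ambience cue that plays small, medium or large-group fire loops according to an Intensity parameter.
# Each tier shares one attenuation asset and has its own pitch and volume modulation.
# A quality-level node per tier plays a random start one-shot mixed with a random loop at the first level, and only a loop from a smaller pool at the others.
# A concurrency override caps simultaneous instances at three.
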SoundCue /Game/Sounds/Fort_GamePlay_Sounds/Curie_Fires/Curie_Fire_Loop_Layered_Cue
root=Switch_0
nodes [
 {
  id=Attenuation_0
  children [Modulator_0]
  properties {attenuationSettings=&/Curie_Fire_Small_01_Att}
 }
 {
  id=Attenuation_1
  children [Modulator_1]
  properties {attenuationSettings=&Curie_Fire_Small_01_Att}
 }
 {
  id=Attenuation_2
  children [Modulator_2]
  properties {attenuationSettings=&Curie_Fire_Small_01_Att}
 }
 {id=Mixer_0, children [Random_6,Random_0], properties {inputVolume [1,1]}}
 {id=Mixer_1, children [Random_7,Random_1], properties {inputVolume [1,1]}}
 {id=Mixer_2, children [Random_8,Random_2], properties {inputVolume [1,1]}}
 {id=Modulator_0, children [QualityLevel_1], properties {pitchMax=1.1, pitchMin=0.8, volumeMax=1, volumeMin=0.85}}
 {id=Modulator_1, children [QualityLevel_2], properties {pitchMax=1.1, pitchMin=0.9, volumeMax=1.1, volumeMin=0.9}}
 {id=Modulator_2, children [QualityLevel_3], properties {pitchMax=1, pitchMin=0.9, volumeMax=1.1, volumeMin=1}}
 {id=QualityLevel_1, children [Mixer_0,Random_3,Random_3,Random_3]}
 {id=QualityLevel_2, children [Mixer_1,Random_4,Random_4,Random_4]}
 {id=QualityLevel_3, children [Mixer_2,Random_5,Random_5,Random_5]}
 {
  id=Random_0
  children [WavePlayer_0, WavePlayer_1, WavePlayer_2, WavePlayer_3, WavePlayer_4, WavePlayer_25]
  properties {weights [1,1,1,1,1,1]}
 }
 {
  id=Random_1
  children [WavePlayer_6, WavePlayer_7, WavePlayer_8, WavePlayer_9, WavePlayer_10, WavePlayer_11]
  properties {weights [1,1,1,1,1,1]}
 }
 {
  id=Random_2
  children [WavePlayer_12, WavePlayer_13, WavePlayer_14, WavePlayer_15, WavePlayer_16, WavePlayer_17]
  properties {weights [1,1,1,1,1,1]}
 }
 {id=Random_3, children [WavePlayer_18, WavePlayer_19, WavePlayer_20], properties {bShouldExcludeFromBranchCulling=true, weights [1,1,1]}}
 {id=Random_4, children [WavePlayer_22, WavePlayer_21], properties {bShouldExcludeFromBranchCulling=true, weights [1,1]}}
 {id=Random_5, children [WavePlayer_23, WavePlayer_24], properties {bShouldExcludeFromBranchCulling=true, weights [1,1]}}
 {
  id=Random_6
  children [WavePlayer_5, WavePlayer_26, WavePlayer_27, WavePlayer_28, WavePlayer_29, WavePlayer_30, WavePlayer_31]
  properties {weights [1,1,1,1,1,1,1]}
 }
 {
  id=Random_7
  children [WavePlayer_32, WavePlayer_33, WavePlayer_34, WavePlayer_35, WavePlayer_36, WavePlayer_37]
  properties {weights [1,1,1,1,1,1]}
 }
 {
  id=Random_8
  children [WavePlayer_38, WavePlayer_39, WavePlayer_40, WavePlayer_41, WavePlayer_42, WavePlayer_43]
  properties {weights [1,1,1,1,1,1]}
 }
 {
  id=Switch_0
  children [
   Attenuation_0
   Attenuation_0
   Attenuation_0
   Attenuation_1
   Attenuation_1
   Attenuation_1
   Attenuation_2
   Attenuation_2
   Attenuation_2
   Attenuation_2
   Attenuation_2
   Attenuation_2
  ]
  properties {intParameterName=Intensity}
 }
 {id=WavePlayer_0, properties {bLooping=true, soundWaveAssetPtr=&/Curie_Fire_SmallLoop_01}}
 {id=WavePlayer_1, properties {bLooping=true, soundWaveAssetPtr=&/Curie_Fire_SmallLoop_02}}
 {id=WavePlayer_10, properties {bLooping=true, soundWaveAssetPtr=&/Curie_Fire_MediumLoop_05}}
 {id=WavePlayer_11, properties {bLooping=true, soundWaveAssetPtr=&/Curie_Fire_MediumLoop_06}}
 {id=WavePlayer_12, properties {bLooping=true, soundWaveAssetPtr=&/Curie_Fire_GroupLargeLoop_01}}
 {id=WavePlayer_13, properties {bLooping=true, soundWaveAssetPtr=&/Curie_Fire_GroupLargeLoop_03}}
 {id=WavePlayer_14, properties {bLooping=true, soundWaveAssetPtr=&/Curie_Fire_GroupLargeLoop_05}}
 {id=WavePlayer_15, properties {bLooping=true, soundWaveAssetPtr=&/Curie_Fire_GroupLargeLoop_02}}
 {id=WavePlayer_16, properties {bLooping=true, soundWaveAssetPtr=&/Curie_Fire_GroupLargeLoop_04}}
 {id=WavePlayer_17, properties {bLooping=true, soundWaveAssetPtr=&/Curie_Fire_GroupLargeLoop_06}}
 {id=WavePlayer_18, properties {bLooping=true, soundWaveAssetPtr=&Curie_Fire_SmallLoop_01}}
 {id=WavePlayer_19, properties {bLooping=true, soundWaveAssetPtr=&/Curie_Fire_SmallLoop_04}}
 {id=WavePlayer_2, properties {bLooping=true, soundWaveAssetPtr=&/Curie_Fire_SmallLoop_03}}
 {id=WavePlayer_20, properties {bLooping=true, soundWaveAssetPtr=&/Curie_Fire_SmallLoop_06}}
 {id=WavePlayer_21, properties {bLooping=true, soundWaveAssetPtr=&Curie_Fire_MediumLoop_05}}
 {id=WavePlayer_22, properties {bLooping=true, soundWaveAssetPtr=&/Curie_Fire_MediumLoop_03}}
 {id=WavePlayer_23, properties {bLooping=true, soundWaveAssetPtr=&Curie_Fire_GroupLargeLoop_01}}
 {id=WavePlayer_24, properties {bLooping=true, soundWaveAssetPtr=&Curie_Fire_GroupLargeLoop_06}}
 {id=WavePlayer_25, properties {bLooping=true, soundWaveAssetPtr=&Curie_Fire_SmallLoop_06}}
 {id=WavePlayer_26, properties {soundWaveAssetPtr=&/Curie_Fire_SmallStart_02}}
 {id=WavePlayer_27, properties {soundWaveAssetPtr=&/Curie_Fire_SmallStart_03}}
 {id=WavePlayer_28, properties {soundWaveAssetPtr=&/Curie_Fire_SmallStart_04}}
 {id=WavePlayer_29, properties {soundWaveAssetPtr=&/Curie_Fire_SmallStart_05}}
 {id=WavePlayer_3, properties {bLooping=true, soundWaveAssetPtr=&Curie_Fire_SmallLoop_04}}
 {id=WavePlayer_30, properties {soundWaveAssetPtr=&/Curie_Fire_SmallStart_06}}
 {id=WavePlayer_31, properties {soundWaveAssetPtr=&/Curie_Fire_SmallStart_07}}
 {id=WavePlayer_32, properties {soundWaveAssetPtr=&/Curie_Fire_MediumStart_06}}
 {id=WavePlayer_33, properties {soundWaveAssetPtr=&/Curie_Fire_MediumStart_01}}
 {id=WavePlayer_34, properties {soundWaveAssetPtr=&/Curie_Fire_MediumStart_02}}
 {id=WavePlayer_35, properties {soundWaveAssetPtr=&/Curie_Fire_MediumStart_03}}
 {id=WavePlayer_36, properties {soundWaveAssetPtr=&/Curie_Fire_MediumStart_04}}
 {id=WavePlayer_37, properties {soundWaveAssetPtr=&/Curie_Fire_MediumStart_05}}
 {id=WavePlayer_38, properties {soundWaveAssetPtr=&/Curie_Fire_GroupLargeStart_06}}
 {id=WavePlayer_39, properties {soundWaveAssetPtr=&/Curie_Fire_GroupLargeStart_01}}
 {id=WavePlayer_4, properties {bLooping=true, soundWaveAssetPtr=&/Curie_Fire_SmallLoop_05}}
 {id=WavePlayer_40, properties {soundWaveAssetPtr=&/Curie_Fire_GroupLargeStart_02}}
 {id=WavePlayer_41, properties {soundWaveAssetPtr=&/Curie_Fire_GroupLargeStart_03}}
 {id=WavePlayer_42, properties {soundWaveAssetPtr=&/Curie_Fire_GroupLargeStart_04}}
 {id=WavePlayer_43, properties {soundWaveAssetPtr=&/Curie_Fire_GroupLargeStart_05}}
 {id=WavePlayer_5, properties {soundWaveAssetPtr=&/Curie_Fire_SmallStart_01}}
 {id=WavePlayer_6, properties {bLooping=true, soundWaveAssetPtr=&/Curie_Fire_MediumLoop_01}}
 {id=WavePlayer_7, properties {bLooping=true, soundWaveAssetPtr=&/Curie_Fire_MediumLoop_02}}
 {id=WavePlayer_8, properties {bLooping=true, soundWaveAssetPtr=&Curie_Fire_MediumLoop_03}}
 {id=WavePlayer_9, properties {bLooping=true, soundWaveAssetPtr=&/Curie_Fire_MediumLoop_04}}
]
properties {
 bOverrideConcurrency=true
 maxDistance=2500
 soundClassObject=&Fort_Ambient_Athena
 volumeMultiplier=0.276
 concurrencyOverrides {maxCount=3}
}
