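# Sound cue for Curie fire on a player, layering a burning loop with ignition and damage sounds, attenuated and limited to one instance per owner.
# A Mixer combines three random picks: a pitch-modulated loop that an envelope fades in, a pitch-modulated small-start sound, and a louder player-damage sound.
# An IsLocal Branch feeds that same mix through a non-critical in-game UI sound class for local players, otherwise the Athena ambient class.
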SoundCue /Game/Sounds/Fort_GamePlay_Sounds/Curie_Fires/Curie_Fire_PlayerLoop_Cue
root=Branch_0
nodes [
 {id=Branch_0, children [SoundClass_0,SoundClass_1,SoundClass_1], properties {boolParameterName=IsLocal}}
 {
  id=Enveloper_0
  children [Modulator_0]
  properties {pitchCurve {editorCurveData {keys [{time=0}]}}, volumeCurve {editorCurveData {keys [<at 0 0>,<at 2 1>]}}}
 }
 {id=Mixer_0, children [Enveloper_0,Modulator_1,Random_2], properties {inputVolume [0.5,1,1.5]}}
 {id=Modulator_0, children [Random_1], properties {pitchMax=1.2, pitchMin=0.9}}
 {id=Modulator_1, children [Random_0], properties {pitchMax=1.2, pitchMin=0.9}}
 {
  id=Random_0
  children [WavePlayer_0, WavePlayer_1, WavePlayer_2, WavePlayer_3, WavePlayer_10, WavePlayer_11, WavePlayer_12]
  properties {weights [1,1,1,1,1,1,1]}
 }
 {id=Random_1, children [WavePlayer_6,WavePlayer_7], properties {weights [1,1]}}
 {id=Random_2, children [WavePlayer_4, WavePlayer_5, WavePlayer_13], properties {weights [1,1,1]}}
 {id=SoundClass_0, children [Mixer_0], properties {soundClassOverride=&Fort_UI_InGame_NonCritical}}
 {id=SoundClass_1, children [Mixer_0], properties {soundClassOverride=&Fort_Ambient_Athena}}
 {id=WavePlayer_0, properties {soundWaveAssetPtr=&/Curie_Fire_SmallStart_01}}
 {id=WavePlayer_1, properties {soundWaveAssetPtr=&/Curie_Fire_SmallStart_02}}
 {id=WavePlayer_10, properties {soundWaveAssetPtr=&/Curie_Fire_SmallStart_05}}
 {id=WavePlayer_11, properties {soundWaveAssetPtr=&/Curie_Fire_SmallStart_06}}
 {id=WavePlayer_12, properties {soundWaveAssetPtr=&/Curie_Fire_SmallStart_07}}
 {id=WavePlayer_13, properties {soundWaveAssetPtr=&/Curie_Fire_PlayerDamage_03}}
 {id=WavePlayer_2, properties {soundWaveAssetPtr=&/Curie_Fire_SmallStart_03}}
 {id=WavePlayer_3, properties {soundWaveAssetPtr=&/Curie_Fire_SmallStart_04}}
 {id=WavePlayer_4, properties {soundWaveAssetPtr=&/Curie_Fire_PlayerDamage_01}}
 {id=WavePlayer_5, properties {soundWaveAssetPtr=&/Curie_Fire_PlayerDamage_02}}
 {id=WavePlayer_6, properties {bLooping=true, soundWaveAssetPtr=&/Curie_Fire_PlayerLoop_01}}
 {id=WavePlayer_7, properties {bLooping=true, soundWaveAssetPtr=&/Curie_Fire_PlayerLoop_03}}
]
properties {
 attenuationSettings=&/Curie_Fire_Small_01_Att
 bOverrideConcurrency=true
 maxDistance=2500
 volumeMultiplier=0.626
 concurrencyOverrides {bLimitToOwner=true, maxCount=1}
}
